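# Niagara cosmetic trail for the FNCS WaywardRebel character skin, combining mesh trails, glow, sparkle sprites and particle lights sampled from skeletal meshes.
# Body emitters pair with "_group" emitters placing two mirrored sets on the head mesh, which respond to BigHeadMode and User.bisHeadHidden.
# Trails pass a time-smoothed position in Interpolate Dynamic, presumably for material lag.
# Time of day tints and dims effects, while owner speed gates sprite spawning.
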NiagaraSystem /BRCosmetics/Characters/Player/Female/Medium/Bodies/F_MED_WaywardRebel/Skins/FNCS/FX/NS_Character_WaywardRebelFNCS_Trail
dependencies {
 usedDataInterfaces [
  &NiagaraDataInterfacePlatformSet
  &NiagaraDataInterfaceArrayFloat3
  &NiagaraDataInterfaceArrayFloat
  &NiagaraDataInterfaceCurve
  &NiagaraDataInterfaceColorCurve
  &NiagaraDataInterfaceSkeletalMesh
 ]
 usedDynamicInputs [&SimulationPosition]
 usedModules [
  &self:SystemSpawnScript.SystemScriptSource.SystemScriptGraph.NiagaraNodeAssignment_0.SetVariables_#1
  &self:SystemSpawnScript.SystemScriptSource.SystemScriptGraph.NiagaraNodeAssignment_1.SetVariables_#2
  &SystemState
  &self:SystemSpawnScript.SystemScriptSource.SystemScriptGraph.NiagaraNodeAssignment_2.SetVariables_#3
  &self:SystemSpawnScript.SystemScriptSource.SystemScriptGraph.NiagaraNodeAssignment_3.SetVariables_#4
 ]
}
emitters {
 Glow {
  emitterUpdateScript [
   &EmitterState(
    Inactive Response=::NewEnumerator1
    Life Cycle Mode=::NewEnumerator1
    Loop Behavior=::NewEnumerator1
    Loop Duration Mode=::NewEnumerator1
    UseLoopDelay=false
   )
   &SpawnBurst_Instantaneous(Use Spawn Probability=true, UseLoopCountLimit=false)
  ]
  particleSpawnScript [
   &InitializeParticle(
    Color=<rgba 1 0.0392 0.0774 0.5>
    Position Mode=::NewEnumerator0
    Sprite Rotation Mode=::NewEnumerator0
    Sprite Size=<5 12.8566>
    Sprite Size Mode=::NewEnumerator4
    Write Material Random=true
   )
   &CameraOffset(Camera Offset Amount=&FN_PickFloatDependOnPlatform(TargetPlatformActive=2, TargetPlatformNoActive=5))
   &SpriteFacingAndAlignment(Sprite Facing=<0 -1 0>, WriteAlignment=false, WriteFacing=true)
   &AddVelocity(Velocity=<0.001 0 0>)
   &ApplyInitialForces()
   /SetVariables(Particles.SIze002:Vector2f=&Vector2DFromFloat(Value=20)) {isSetParameters=true}
  ]
  particleUpdateScript [
   &ParticleState(Kill Particles When Lifetime Has Elapsed=false, Let Infinitely Lived Particles Die When Emitter Deactivates=true)
   &ScaleColor(
    Scale Alpha=&Multiply_Float(
     A=$System.AlphaFade
     B=&FNTimeOfDayScalar(Night Scale=0.8, Morning Scale=0.6, Evening Scale=0.8)
    )
    Scale RGB=&VectorFromFloat(Value=&FN_PickFloatDependOnPlatform(TargetPlatformActive=1, TargetPlatformNoActive=0.2))
    ScaleA=true
    ScaleRGB=true
   )
   &SkeletalMeshLocation(Bone Sampling Mode=::NewEnumerator4, MaskSpawnGroup=false, Sampled Position Offset=<2.3848 -6.2938 0>, Skeletal Mesh=$System.Skeletal Mesh)
  ]
  rendererClasses [&NiagaraSpriteRendererProperties]
  renderers [
   {rendererClass=&NiagaraSpriteRendererProperties}
   {rendererClass=&NiagaraSpriteRendererProperties, rendererIndex=1}
  ]
 }
 Light {
  emitterUpdateScript [
   &EmitterState(
    Inactive Response=::NewEnumerator1
    Life Cycle Mode=::NewEnumerator1
    Loop Behavior=::NewEnumerator1
    Loop Duration Mode=::NewEnumerator1
    UseLoopDelay=false
   )
   &SpawnBurst_Instantaneous(Use Spawn Probability=false, UseLoopCountLimit=false)
  ]
  particleSpawnScript [&InitializeParticle(Color Mode=::NewEnumerator0, Lifetime=1.5, UsePositionOffset=false, Write Material Random=true)]
  particleUpdateScript [
   &ParticleState(
    Kill Particles When Lifetime Has Elapsed=false
    Let Infinitely Lived Particles Die When Emitter Deactivates=true
    Loop Particles Lifetime=true
   )
   &SkeletalMeshLocation(Bone Sampling Mode=::NewEnumerator4, MaskSpawnGroup=false, Skeletal Mesh=$System.Skeletal Mesh)
   &FN_Light_Attributes(
    Color=&MultiplyLinearColor(
     Linear Color=&ColorFromCurve(ColorCurve=DataInterface() {propertyValues {
        AlphaCurve {keys []}
        BlueCurve {keys [
          <at 0 0>
          <at 0.2097 0.0529>
          <at 0.5678 1>
          {arriveTangent=-2.3138, interpMode=RCIM_Cubic, leaveTangent=-2.3138, time=0.846, value=0.5539}
          <at 1 0>
         ]}
        GreenCurve {keys [<at 0 0>, <at 0.2097 0>, <at 0.5678 1>, {arriveTangent=-2.3138, interpMode=RCIM_Cubic, leaveTangent=-2.3138, time=0.846, value=0}, <at 1 0>]}
        RedCurve {keys [<at 0 0>, <at 0.2097 1>, <at 0.5678 1>, {arriveTangent=-2.3138, interpMode=RCIM_Cubic, leaveTangent=-2.3138, time=0.846, value=1}, <at 1 0>]}
       }})
     Scale Factor=&MakeVector4_FromLinearColor(LinearColor=$System.TOD)
    )
    Create Color Attribute=true
    Create Exponent Attribute=true
    Create Position Attribute=true
    Create Radius Attribute=true
    Exponent=1.5
    Position=&AddVectorToPosition(Position=$Particles.Position, Vector=&TransformVector(Vector=<0 0 4>, Destination Space=::Simulation))
    Radius=12
    Scale Color=&Multiply_Float(A=20, B=$System.AlphaFade)
   )
  ]
  rendererClasses [&NiagaraLightRendererProperties]
  renderers [{rendererClass=&NiagaraLightRendererProperties, rendererIndex=0}]
 }
 Light_group {
  emitterUpdateScript [
   &EmitterState(
    Inactive Response=::NewEnumerator1
    Life Cycle Mode=::NewEnumerator1
    Loop Behavior=::NewEnumerator1
    Loop Duration Mode=::NewEnumerator1
    UseLoopDelay=false
   )
   &SpawnBurst_Instantaneous(Spawn Count=2, Use Spawn Probability=false, UseLoopCountLimit=false)
  ]
  particleSpawnScript [
   &PartitionParticles()
   &InitializeParticle(Color Mode=::NewEnumerator0, Lifetime=1.5, Sprite Size Mode=::NewEnumerator3, UsePositionOffset=false, Write Material Random=true)
  ]
  particleUpdateScript [
   &ParticleState(
    Kill Particles When Lifetime Has Elapsed=false
    Let Infinitely Lived Particles Die When Emitter Deactivates=true
    Loop Particles Lifetime=true
   )
   &SkeletalMeshLocation(
    Bone Sampling Mode=::NewEnumerator4
    MaskSpawnGroup=false
    Sampled Position Offset=&Multiply_VectorByFloat(
     Vector=&SelectVectorFromArray(
      Direct Array Index=$Particles.PartitionParticles.Partition
      Vector Selection Array=DataInterface() {propertyValues {FloatData [<3 0 15>, <3 0 -15>]}}
     )
     Float=&Lerp_Float(A=1, B=2, Alpha=$NPC.FortniteNPC.BigHeadMode)
    )
    Skeletal Mesh=$System.Skeletal Mesh_head
   )
   &FN_Light_Attributes(
    Color=&MultiplyLinearColor(
     Linear Color=&ColorFromCurve(ColorCurve=DataInterface() {propertyValues {
        AlphaCurve {keys []}
        BlueCurve {keys [
          <at 0 0>
          <at 0.2097 0.0529>
          <at 0.5678 1>
          {arriveTangent=-2.3138, interpMode=RCIM_Cubic, leaveTangent=-2.3138, time=0.846, value=0.5539}
          <at 1 0>
         ]}
        GreenCurve {keys [<at 0 0>, <at 0.2097 0>, <at 0.5678 1>, {arriveTangent=-2.3138, interpMode=RCIM_Cubic, leaveTangent=-2.3138, time=0.846, value=0}, <at 1 0>]}
        RedCurve {keys [<at 0 0>, <at 0.2097 1>, <at 0.5678 1>, {arriveTangent=-2.3138, interpMode=RCIM_Cubic, leaveTangent=-2.3138, time=0.846, value=1}, <at 1 0>]}
       }})
     Scale Factor=&MakeVector4_FromLinearColor(LinearColor=$System.TOD)
    )
    Create Color Attribute=true
    Create Exponent Attribute=true
    Create Position Attribute=true
    Create Radius Attribute=true
    Enable Particle Light Enable=true
    Exponent=1.5
    Light Enabled:NiagaraBool=&Boolean_NotOperation(Bool=$User.bisHeadHidden)
    Position=&AddVectorToPosition(
     Position=$Particles.Position
     Vector=&TransformVector(
      Vector=&SelectVectorFromArray(
       Direct Array Index=$Particles.PartitionParticles.Partition
       Vector Selection Array=DataInterface() {propertyValues {FloatData [<0 0 0>, <0 0 0>]}}
      )
      Destination Space=::Simulation
     )
    )
    Radius=&Lerp_Float(A=15, B=30, Alpha=$NPC.FortniteNPC.BigHeadMode)
    Scale Color=&Multiply_Float(A=20, B=$System.AlphaFade)
   )
  ]
  rendererClasses [&NiagaraLightRendererProperties]
  renderers [{rendererClass=&NiagaraLightRendererProperties, rendererIndex=0}]
 }
 Sprites {
  emitterUpdateScript [
   &EmitterState(
    Enable Distance Culling=true
    Inactive Response=::NewEnumerator1
    Life Cycle Mode=::NewEnumerator1
    Loop Delay=&RandomRangeFloat(Minimum=0, Maximum=1)
    Max Distance Response=::KillImmediately
    Module.bUseMaxDistance=true
    Module.bUseMinDistance=false
    Scalability Mode=::NewEnumerator1
    UseLoopDelay=false
   )
   &SpawnRate(
    Spawn Probability=&MakeCustomFloatFromBool(
     Bool=&SetBoolByFloatComparison(A=&VectorLength(Vector=$Engine.Owner.Velocity), B=1000)
     True Float=0
     False Float=1
    )
    SpawnRate=12
    Use Spawn Probability=true
   )
  ]
  particleSpawnScript [
   &InitializeParticle(
    Color=&MergeHelper_ColorSwitch(
     0=<rgba 1 0.0008 0.0201 1>
     1=<rgba 0.6354 0.0997 0.8698 1>
     2=<rgba 0.6458 0.6458 0.6458 1>
     3=<rgba 0.0156 0.0156 0.0156 1>
     Index=&ModuloInt(A=&ReturnParticleIDAttributes(), B=4)
     Num=::NewEnumerator2
    )
    Lifetime Max=0.45
    Lifetime Min=0.4
    Lifetime Mode=::NewEnumerator1
    Sprite Rotation Mode=::NewEnumerator2
    Sprite Size Mode=::NewEnumerator1
    Uniform Sprite Size Max=8
    Uniform Sprite Size Min=4
    UsePositionOffset=false
    Write Material Random=true
   )
   &ShapeLocation(
    Cylinder Height=12
    Cylinder Radius=6
    MaskSpawnGroup=false
    Override Seed=false
    Rotation Axis=<0 1 0>
    Rotation Mode=::NewEnumerator1
    Shape Origin=&VectorFromFloat(Value=0)
    Shape Primitive=::NewEnumerator1
    Surface Only=false
    UseOwnerScale=true
   )
   /SetVariables(Particles.VisibilityTag=&RandomRangeInt(Minimum=0, Maximum=3)) {isSetParameters=true, module=SetVariables_#5}
   /SetVariables(Particles.Color_triangleglow:LinearColor=&MultiplyLinearColor(
     Linear Color=$Particles.Color
     Scale Factor=&MakeVector4_FromLinearColor(LinearColor=/Niagara/DynamicInputs/TypeConversions/MakeLinearColorFromFloat(Value:NiagaraFloat=0.5))
    )) {isSetParameters=true, module=SetVariables_#6}
  ]
  particleUpdateScript [
   &ParticleState()
   &SkeletalMeshLocation(
    Bone Sampling Mode=::NewEnumerator4
    MaskSpawnGroup=false
    Position Sampling=::NewEnumerator1
    Sampled Position Offset=&AddVectorToPosition(Position=$Particles.ShapeLocation.ShapePosition, Vector=<0 -20 0>)
    Skeletal Mesh=$System.Skeletal Mesh
    Source Movement Scale:Vector3f=<0.9 0.9 0.9>
   )
   &ScaleSpriteSize(Scale Factor=&Vector2DFromFloat(Value=&FloatFromCurve(FloatCurve=DataInterface() {propertyValues {Curve {keys [
          {interpMode=RCIM_Cubic, tangentMode=RCTM_User, time=0, value=0}
          {interpMode=RCIM_Cubic, leaveTangent=-1.6234, tangentMode=RCTM_Break, time=0.4, value=1}
          {tangentMode=RCTM_User, time=1, value=0}
         ]}}})))
   /SetVariables(Particles.Size002:Vector2f=&Multiply_Vector2D(A=$Particles.SpriteSize, B=&Vector2DFromFloat(Value=1.3))) {isSetParameters=true, module=SetVariables_#7}
   /SetVariables(Particles.Size_Triangle:Vector2f=&Multiply_Vector2D(A=$Particles.SpriteSize, B=&Vector2DFromFloat(Value=0.7))) {isSetParameters=true, module=SetVariables_#8}
   &Drag(Use Linear Drag=true, Use Rotational Drag=true)
   &SolveForcesAndVelocity(Clamp Velocity=false, Limit Acceleration=false)
   &ScaleColor(
    Scale Alpha=&Multiply_Float(
     A=&FNTimeOfDayScalar(Night Scale=1, Morning Scale=0.3, Evening Scale=0.3)
     B=&FloatFromCurve(FloatCurve=DataInterface() {propertyValues {Curve {keys [
          {interpMode=RCIM_Cubic, tangentMode=RCTM_User, time=0.1, value=0}
          {interpMode=RCIM_Cubic, tangentMode=RCTM_User, time=0.3, value=0.8}
          {tangentMode=RCTM_User, time=0.8, value=0}
         ]}}})
    )
    Scale RGB=&MakeVectorFromLinearColorRGB(Color=$System.TOD)
    ScaleA=true
    ScaleRGB=true
   )
   /SetVariables(Particles.Color002=&MultiplyLinearColor(
     Linear Color=&MultiplyLinearColor(Linear Color=<rgba 1.2 1.2 1.2 1>, Scale Factor=&MakeVector4_FromLinearColor(LinearColor=$System.TOD))
     Scale Factor=&MakeVector4_FromLinearColor(LinearColor=$Particles.Color)
    )) {isSetParameters=true, module=SetVariables_#9}
  ]
  rendererClasses [&NiagaraSpriteRendererProperties]
  renderers [
   {rendererClass=&NiagaraSpriteRendererProperties}
   {rendererClass=&NiagaraSpriteRendererProperties, rendererIndex=1}
   {rendererClass=&NiagaraSpriteRendererProperties, rendererIndex=2}
   {rendererClass=&NiagaraSpriteRendererProperties, rendererIndex=3}
  ]
 }
 Sprites_group {
  emitterUpdateScript [
   &EmitterState(
    Enable Distance Culling=true
    Inactive Response=::NewEnumerator1
    Life Cycle Mode=::NewEnumerator1
    Loop Delay=&RandomRangeFloat(Minimum=0, Maximum=1)
    Max Distance Response=::KillImmediately
    Module.bUseMaxDistance=true
    Module.bUseMinDistance=false
    Scalability Mode=::NewEnumerator1
    UseLoopDelay=false
   )
   &SpawnRate(
    Spawn Probability=&MakeCustomFloatFromBool(
     Bool=&SetBoolByFloatComparison(A=&VectorLength(Vector=$Engine.Owner.Velocity), B=1000)
     True Float=0
     False Float=1
    )
    SpawnRate=&MakeCustomFloatFromBool(Bool=$User.bisHeadHidden, True Float=0, False Float=12)
    Use Spawn Probability=true
   )
  ]
  particleSpawnScript [
   &PartitionParticles()
   &InitializeParticle(
    Color=&MergeHelper_ColorSwitch(
     0=<rgba 1 0.0208 0.0216 1>
     1=<rgba 0.6354 0.0997 0.8698 1>
     2=<rgba 0.6458 0.6458 0.6458 1>
     3=<rgba 0.0156 0.0156 0.0156 1>
     Index=&ModuloInt(A=&ReturnParticleIDAttributes(), B=4)
     Num=::NewEnumerator2
    )
    Lifetime Max=0.36
    Lifetime Min=0.35
    Lifetime Mode=::NewEnumerator1
    Sprite Rotation Mode=::NewEnumerator2
    Sprite Size Mode=::NewEnumerator1
    Uniform Sprite Size Max=5
    Uniform Sprite Size Min=4
    UsePositionOffset=false
    Write Material Random=true
   )
   &InitializeParticle(
    Color=&MakeCustomLinearColorFromBool(
     Bool=&RandomBool()
     True Color=<rgba 1 0.0365 0.0867 1>
     False Color=&MakeCustomLinearColorFromBool(Bool=&RandomBool(), True Color=<rgba 0.6354 0.0997 0.8698 1>, False Color=<rgba 0.6458 0.6458 0.6458 1>)
    )
    Lifetime Max=0.36
    Lifetime Min=0.35
    Lifetime Mode=::NewEnumerator1
    Sprite Rotation Mode=::NewEnumerator2
    Sprite Size Mode=::NewEnumerator1
    Uniform Sprite Size Max=5
    Uniform Sprite Size Min=4
    UsePositionOffset=false
    Write Material Random=true
   ) {enabled=false, module=InitializeParticle001}
   &ShapeLocation(
    MaskSpawnGroup=false
    Override Seed=false
    Rotation Axis=<0 1 0>
    Rotation Mode=::NewEnumerator1
    Shape Origin=&VectorFromFloat(Value=0)
    Sphere Radius=5
    UseOwnerScale=true
   )
   /SetVariables(Particles.VisibilityTag=&RandomRangeInt(Minimum=0, Maximum=3)) {isSetParameters=true, module=SetVariables_#5}
   /SetVariables(Particles.Color_triangleglow:LinearColor=&MultiplyLinearColor(
     Linear Color=$Particles.Color
     Scale Factor=&MakeVector4_FromLinearColor(LinearColor=&MakeLinearColorFromFloat(Value:NiagaraFloat=0.5))
    )) {isSetParameters=true, module=SetVariables_#6}
  ]
  particleUpdateScript [
   &ParticleState()
   &SkeletalMeshLocation(
    Bone Sampling Mode=::NewEnumerator4
    MaskSpawnGroup=false
    Sampled Position Offset=&AddVectorToPosition(
     Position=$Particles.ShapeLocation.ShapePosition
     Vector=&SelectVectorFromArray(
      Direct Array Index=$Particles.PartitionParticles.Partition
      Vector Selection Array=DataInterface() {propertyValues {FloatData [<5 -8 -13>, <4 -8 15>]}}
     )
    )
    Skeletal Mesh=$System.Skeletal Mesh_head
   )
   &ScaleSpriteSize(Scale Factor=&Vector2DFromFloat(Value=&FloatFromCurve(
      FloatCurve=DataInterface() {propertyValues {Curve {keys [
          {interpMode=RCIM_Cubic, tangentMode=RCTM_User, time=0, value=0}
          {interpMode=RCIM_Cubic, tangentMode=RCTM_User, time=0.4, value=1}
          {tangentMode=RCTM_User, time=1, value=0}
         ]}}}
      Scale Curve=&Lerp_Float(A=1, B=2, Alpha=$NPC.FortniteNPC.BigHeadMode)
     )))
   /SetVariables(Particles.Size002:Vector2f=&Multiply_Vector2D(A=$Particles.SpriteSize, B=&Vector2DFromFloat(Value=1.3))) {isSetParameters=true, module=SetVariables_#7}
   /SetVariables(Particles.Size_Triangle:Vector2f=&Multiply_Vector2D(A=$Particles.SpriteSize, B=&Vector2DFromFloat(Value=0.6))) {isSetParameters=true, module=SetVariables_#8}
   &Drag(Use Linear Drag=true, Use Rotational Drag=true)
   &SolveForcesAndVelocity(Clamp Velocity=false, Limit Acceleration=false)
   &ScaleColor(
    Scale Alpha=&Multiply_Float(
     A=&FNTimeOfDayScalar(Night Scale=1, Morning Scale=0.3, Evening Scale=0.3)
     B=&FloatFromCurve(FloatCurve=DataInterface() {propertyValues {Curve {keys [
          {interpMode=RCIM_Cubic, tangentMode=RCTM_User, time=0.1, value=0}
          {interpMode=RCIM_Cubic, tangentMode=RCTM_User, time=0.3, value=0.8}
          {tangentMode=RCTM_User, time=0.8, value=0}
         ]}}})
    )
    Scale RGB=&MakeVectorFromLinearColorRGB(Color=$System.TOD)
    ScaleA=true
    ScaleRGB=true
   )
   /SetVariables(Particles.Color002=&MultiplyLinearColor(
     Linear Color=&MultiplyLinearColor(Linear Color=<rgba 1.2 1.2 1.2 1>, Scale Factor=&MakeVector4_FromLinearColor(LinearColor=$System.TOD))
     Scale Factor=&MakeVector4_FromLinearColor(LinearColor=$Particles.Color)
    )) {isSetParameters=true, module=SetVariables_#9}
   &ScaleSpriteSize(
    Scale Sprite Size Mode=::NewEnumerator0
    Uniform Scale Factor=&MakeCustomFloatFromBool(Bool=$User.bisHeadHidden, True Float=0, False Float=1)
   ) {module=ScaleSpriteSize001}
  ]
  rendererClasses [&NiagaraSpriteRendererProperties]
  renderers [
   {rendererClass=&NiagaraSpriteRendererProperties}
   {rendererClass=&NiagaraSpriteRendererProperties, rendererIndex=1}
   {rendererClass=&NiagaraSpriteRendererProperties, rendererIndex=2}
   {rendererClass=&NiagaraSpriteRendererProperties, rendererIndex=3}
  ]
 }
 Trail {
  emitterUpdateScript [
   &EmitterState(
    Inactive Response=::NewEnumerator1
    Life Cycle Mode=::NewEnumerator1
    Loop Behavior=::NewEnumerator1
    Loop Duration Mode=::NewEnumerator1
    UseLoopDelay=false
   )
   &SpawnBurst_Instantaneous(Use Spawn Probability=false, UseLoopCountLimit=false)
  ]
  particleSpawnScript [
   &InitializeParticle(Color=<rgba 0.0573 0.0573 0.0573 7>, Lifetime=1.5, Mesh Scale Mode=::NewEnumerator4, UsePositionOffset=true, Write Material Random=true)
   &InitialMeshOrientation(Mesh Orientation Mode=::NewEnumerator2, Rotation=<0 -0.25 0.01>, Use Rotation Vector=true)
   /SetVariables(Particles.Size_Orange:Vector3f=<0.5 0.3 0.7>) {isSetParameters=true, module=SetVariables_#10}
   /SetVariables(Particles.Size_Blue:Vector3f=<0.5 0.3 1>) {isSetParameters=true}
   /SetVariables(Particles.Size_D:Vector3f=<0.6 0.4 0.8>) {isSetParameters=true, module=SetVariables_#11}
   /SetVariables(Particles.2ndColor:LinearColor=<rgba 1 1 1 5>, Particles.Color_D:LinearColor=<rgba 0.115 0.0418 0.1823 5>) {isSetParameters=true, module=SetVariables_#12}
   /SetVariables(Particles.Size_Green:Vector3f=<0.3 0.3 0.5>) {isSetParameters=true, module=SetVariables_#13}
  ]
  particleUpdateScript [
   &ParticleState(Kill Particles When Lifetime Has Elapsed=false, Let Infinitely Lived Particles Die When Emitter Deactivates=true)
   &SkeletalMeshLocation(Bone Sampling Mode=::NewEnumerator4, MaskSpawnGroup=false, Skeletal Mesh=$System.Skeletal Mesh)
   &InterpolateOverTime(Rate Of Change=8, Target Vector Value:Vector3f=$Particles.SkeletalMeshLocation.SampledPosition, Variable Type=::NewEnumerator2)
   /SetVariables(Particles.Interpolate Dynamic=&MakeFloat4FromVectorAndFloat("Vector (RGB)"=$Particles.InterpolateOverTime.InterpolateOverTime_V3_.Moving Average, "Float (Alpha)"=0.5)) {isSetParameters=true}
   &ScaleColor(
    Scale Alpha=$System.AlphaFade
    Scale RGB=&MakeVectorFromLinearColorRGB(Color=$System.TOD)
    ScaleA=true
    ScaleRGB=true
   )
   /SetVariables(Particles.Color_D:LinearColor=&MultiplyLinearColor(
     Linear Color=$Particles.Initial.Color_D
     Scale Factor=&MakeFloat4FromVectorAndFloat("Vector (RGB)"=&MakeVectorFromLinearColorRGB(Color=$System.TOD), "Float (Alpha)"=$System.AlphaFade)
    )) {isSetParameters=true}
  ]
  rendererClasses [&NiagaraMeshRendererProperties]
  renderers [
   {rendererClass=&NiagaraMeshRendererProperties, rendererIndex=0}
   {rendererClass=&NiagaraMeshRendererProperties, rendererIndex=1}
   {rendererClass=&NiagaraMeshRendererProperties, rendererIndex=2}
   {rendererClass=&NiagaraMeshRendererProperties, rendererIndex=3}
  ]
 }
 Trail_group {
  emitterUpdateScript [
   &EmitterState(
    Inactive Response=::NewEnumerator1
    Life Cycle Mode=::NewEnumerator1
    Loop Behavior=::NewEnumerator1
    Loop Duration Mode=::NewEnumerator1
    UseLoopDelay=false
   )
   &SpawnBurst_Instantaneous(Spawn Count=2, Use Spawn Probability=false, UseLoopCountLimit=false)
  ]
  particleSpawnScript [
   &PartitionParticles()
   &InitializeParticle(Color=<rgba 1 1 1 5>, Lifetime=1.5, Mesh Scale Mode=::NewEnumerator4, UsePositionOffset=true, Write Material Random=true)
   &CameraOffset(Camera Offset Amount=&FN_PickFloatDependOnPlatform(TargetPlatformActive=0, TargetPlatformNoActive=5))
   &InitialMeshOrientation(
    Mesh Orientation Mode=::NewEnumerator2
    Rotation=&SelectVectorFromArray(
     Direct Array Index=$Particles.PartitionParticles.Partition
     Vector Selection Array=DataInterface() {propertyValues {FloatData [<0 0.25 0>, <0 0.25 0>]}}
    )
    Use Rotation Vector=true
   ) {module=InitialMeshOrientation001}
   &InitialMeshOrientation(
    Mesh Orientation Mode=::NewEnumerator2
    Rotation=&SelectVectorFromArray(
     Direct Array Index=$Particles.PartitionParticles.Partition
     Vector Selection Array=DataInterface() {propertyValues {FloatData [<0.1 0 0>, <-0.1 0 0.05>]}}
    )
    Use Rotation Vector=true
   ) {module=InitialMeshOrientation002}
   /SetVariables(Particles.Size_Orange:Vector3f=&SelectVectorFromArray(
     Direct Array Index=$Particles.PartitionParticles.Partition
     Vector Selection Array=DataInterface() {propertyValues {FloatData [<-0.3 0.2 0.55>, <0.3 0.2 0.65>]}}
    )) {isSetParameters=true, module=SetVariables_#10}
   /SetVariables(Particles.Size_D:Vector3f=&SelectVectorFromArray(
     Direct Array Index=$Particles.PartitionParticles.Partition
     Vector Selection Array=DataInterface() {propertyValues {FloatData [<-0.3 0.25 0.55>, <0.3 0.25 0.55>]}}
    )) {isSetParameters=true, module=SetVariables_#11}
   /SetVariables(Particles.Color_D:LinearColor=<rgba 0.115 0.0418 0.1823 5>) {isSetParameters=true, module=SetVariables_#12}
   /SetVariables(Particles.Size_Green:Vector3f=&SelectVectorFromArray(
     Direct Array Index=$Particles.PartitionParticles.Partition
     Vector Selection Array=DataInterface() {propertyValues {FloatData [<0.25 0.2 0.4>, <0.25 0.2 0.4>]}}
    )) {isSetParameters=true, module=SetVariables_#13}
  ]
  particleUpdateScript [
   &ParticleState(Kill Particles When Lifetime Has Elapsed=false, Let Infinitely Lived Particles Die When Emitter Deactivates=true)
   /SetVariables(
    Particles.Size_D:Vector3f=&Multiply_VectorByFloat(Vector=$Particles.Initial.Size_D, Float=&Lerp_Float(A=1, B=2, Alpha=$NPC.FortniteNPC.BigHeadMode))
    Particles.Size_Green:Vector3f=&Multiply_VectorByFloat(Vector=$Particles.Initial.Size_Green, Float=&Lerp_Float(A=1, B=2, Alpha=$NPC.FortniteNPC.BigHeadMode))
    Particles.Size_Orange:Vector3f=&Multiply_VectorByFloat(Vector=$Particles.Initial.Size_Orange, Float=&Lerp_Float(A=1, B=2, Alpha=$NPC.FortniteNPC.BigHeadMode))
   ) {isSetParameters=true}
   &SkeletalMeshLocation(
    Bone Sampling Mode=::NewEnumerator4
    MaskSpawnGroup=false
    Sampled Position Offset=&Multiply_VectorByFloat(
     Vector=&SelectVectorFromArray(
      Direct Array Index=$Particles.PartitionParticles.Partition
      Vector Selection Array=DataInterface() {propertyValues {FloatData [<2 0 -10>, <2 0 11>]}}
     )
     Float=&Lerp_Float(A=1, B=2, Alpha=$NPC.FortniteNPC.BigHeadMode)
    )
    Skeletal Mesh=$System.Skeletal Mesh_head
   )
   &DynamicMaterialParameters(
    Index 0 Param 1=&RandomRangeFloat(Minimum=-0.6, Maximum=-1.2)
    Index 0 Param 2=&SelectFloatFromArray(Direct Array Index=$Particles.PartitionParticles.Partition, Float Selection Array=DataInterface() {propertyValues {FloatData [0,0.5]}})
    Index 0 Param 3=-50
    Index 0 Param 4=50
    Module.Index0Param0Write=true
    Module.Index0Param1Write=true
    Module.Index0Param2Write=true
    Module.Index0Param3Write=true
   )
   &InterpolateOverTime(Rate Of Change=10, Target Vector Value:NiagaraPosition=$Particles.Position, Variable Type=::NewEnumerator2)
   /SetVariables(Particles.Interpolate Dynamic=&MakeFloat4FromVectorAndFloat("Vector (RGB)"=$Particles.InterpolateOverTime.InterpolateOverTime_V3_.Moving Average, "Float (Alpha)"=0.5)) {isSetParameters=true}
   &ScaleColor(
    Scale Alpha=$System.AlphaFade
    Scale RGB=&MakeVectorFromLinearColorRGB(Color=$System.TOD)
    ScaleA=true
    ScaleRGB=true
   )
   /SetVariables(Particles.Color_D:LinearColor=&MultiplyLinearColor(
     Linear Color=$Particles.Initial.Color_D
     Scale Factor=&MakeFloat4FromVectorAndFloat("Vector (RGB)"=&MakeVectorFromLinearColorRGB(Color=$System.TOD), "Float (Alpha)"=$System.AlphaFade)
    )) {isSetParameters=true}
   /SetVariables(Particles.VisibilityTag=&MakeCustomIntFromBool(Bool=$User.bisHeadHidden, True INT=1, False INT=0)) {isSetParameters=true}
  ]
  rendererClasses [&NiagaraMeshRendererProperties]
  renderers [
   {rendererClass=&NiagaraMeshRendererProperties, rendererIndex=0}
   {rendererClass=&NiagaraMeshRendererProperties, rendererIndex=1}
   {rendererClass=&NiagaraMeshRendererProperties, rendererIndex=2}
  ]
 }
 Trail_soft {
  emitterUpdateScript [
   &EmitterState(
    Inactive Response=::NewEnumerator1
    Life Cycle Mode=::NewEnumerator1
    Loop Behavior=::NewEnumerator1
    Loop Duration Mode=::NewEnumerator1
    Reset Age On Awaken=true
    Scalability Mode=::NewEnumerator1
    UseLoopDelay=false
   )
   &SpawnBurst_Instantaneous(Spawn Count=1, Use Spawn Probability=false)
  ]
  particleSpawnScript [
   &InitializeParticle(
    Color=<rgba 0.4427 0.03 0.0426 1>
    Mass Mode=::NewEnumerator1
    Mesh Scale=<0.5 0.5 1.6>
    Mesh Scale Mode=::NewEnumerator4
    Sprite Size Mode=::NewEnumerator4
    UsePositionOffset=true
    Write Material Random=true
   ) {module=InitializeParticle001}
   &DynamicMaterialParameters(
    Index 0 Param 1=0.5
    Index 0 Param 2=1
    Index 0 Param 3=20
    Module.Param0WriteEnabled=true
    Module.Param1WriteEnabled=true
    Module.Param2WriteEnabled=true
    Module.Param3WriteEnabled=true
   )
   &InitialMeshOrientation(Mesh Orientation Mode=::NewEnumerator2, Rotation=<0.1574 0.25 -0.1944>, Use Rotation Vector=true) {module=InitialMeshOrientation001}
  ]
  particleUpdateScript [
   &ParticleState(Kill Particles When Lifetime Has Elapsed=false, Let Infinitely Lived Particles Die When Emitter Deactivates=true)
   &SkeletalMeshLocation(
    Bone Sampling Mode=::NewEnumerator4
    "Interpolate Mesh Sampling (Update)"=true
    MaskSpawnGroup=false
    Sampled Position Offset=<2 0 0>
    Skeletal Mesh=$System.Skeletal Mesh
   ) {module=SkeletalMeshLocation001}
   &InterpolateOverTime(Rate Of Change=5, Target Vector Value:NiagaraPosition=$Particles.Position, Variable Type=::NewEnumerator2)
   /SetVariables(Particles.Interpolate Dynamic=&MakeFloat4FromVectorAndFloat("Vector (RGB)"=$Output.InterpolateOverTime.Moving Average, "Float (Alpha)"=1)) {isSetParameters=true, module=SetVariables_#14}
   &SolveForcesAndVelocity(Clamp Velocity=false, Limit Acceleration=false)
   &ScaleColor(
    Scale Alpha=&Multiply_Float(
     A=$System.AlphaFade
     B=&Multiply_Float(A=&FNTimeOfDayScalar(Night Scale=1, Morning Scale=1, Evening Scale=1), B=0.05)
    )
    Scale RGB=&MakeVectorFromLinearColorRGB(Color=$System.TOD)
    ScaleA=true
    ScaleRGB=true
   )
  ]
  rendererClasses [&NiagaraMeshRendererProperties]
  renderers [{rendererClass=&NiagaraMeshRendererProperties, rendererIndex=0}]
 }
 Trail_soft_group {
  emitterUpdateScript [
   &EmitterState(
    Inactive Response=::NewEnumerator1
    Life Cycle Mode=::NewEnumerator1
    Loop Behavior=::NewEnumerator1
    Loop Duration Mode=::NewEnumerator1
    Reset Age On Awaken=true
    Scalability Mode=::NewEnumerator1
    UseLoopDelay=false
   )
   &SpawnBurst_Instantaneous(Spawn Count=2, Use Spawn Probability=false)
  ]
  particleSpawnScript [
   &PartitionParticles()
   &InitializeParticle(
    Color=<rgba 0.5469 0.0142 0.0247 2>
    Mass Mode=::NewEnumerator1
    Mesh Scale=<1.2 0.1 1>
    Mesh Scale Mode=::NewEnumerator4
    Sprite Size Mode=::NewEnumerator4
    UsePositionOffset=true
    Write Material Random=true
   ) {module=InitializeParticle001}
   &DynamicMaterialParameters(
    Index 0 Param 1=0.5
    Index 0 Param 2=1
    Index 0 Param 3=15
    Module.Param0WriteEnabled=true
    Module.Param1WriteEnabled=true
    Module.Param2WriteEnabled=true
    Module.Param3WriteEnabled=true
   )
   &InitialMeshOrientation(
    Mesh Orientation Mode=::NewEnumerator2
    Rotation=&SelectVectorFromArray(
     Direct Array Index=$Particles.PartitionParticles.Partition
     Vector Selection Array=DataInterface() {propertyValues {FloatData [<0.28 0 0>, <0.2 0 0>]}}
    )
    Use Rotation Vector=true
   ) {module=InitialMeshOrientation001}
   &InitialMeshOrientation(
    Mesh Orientation Mode=::NewEnumerator2
    Rotation=&SelectVectorFromArray(
     Direct Array Index=$Particles.PartitionParticles.Partition
     Vector Selection Array=DataInterface() {propertyValues {FloatData [<0.05 0 0>, <-0.05 0 0.02>]}}
    )
    Use Rotation Vector=true
   )
  ]
  particleUpdateScript [
   &ParticleState(Kill Particles When Lifetime Has Elapsed=false, Let Infinitely Lived Particles Die When Emitter Deactivates=true)
   &SkeletalMeshLocation(
    Bone Sampling Mode=::NewEnumerator4
    "Interpolate Mesh Sampling (Update)"=true
    MaskSpawnGroup=false
    Sampled Position Offset=&Multiply_VectorByFloat(
     Vector=&SelectVectorFromArray(
      Direct Array Index=$Particles.PartitionParticles.Partition
      Vector Selection Array=DataInterface() {propertyValues {FloatData [<2 0 -10>, <2 0 11>]}}
     )
     Float=&Lerp_Float(A=1, B=2, Alpha=$NPC.FortniteNPC.BigHeadMode)
    )
    Skeletal Mesh=$System.Skeletal Mesh_head
   ) {module=SkeletalMeshLocation001}
   &ScaleMeshSize(Scale Factor=&VectorFromFloat(Value=&Lerp_Float(A=1, B=2, Alpha=$NPC.FortniteNPC.BigHeadMode))) {module=ScaleMeshSize001}
   &InterpolateOverTime(Rate Of Change=7, Target Vector Value:NiagaraPosition=$Particles.Position, Variable Type=::NewEnumerator2)
   /SetVariables(Particles.Interpolate Dynamic=&MakeFloat4FromVectorAndFloat("Vector (RGB)"=$Output.InterpolateOverTime.Moving Average, "Float (Alpha)"=1)) {isSetParameters=true, module=SetVariables_#14}
   &SolveForcesAndVelocity(Clamp Velocity=false, Limit Acceleration=false)
   &ScaleColor(
    Scale Alpha=&Multiply_Float(
     A=$System.AlphaFade
     B=&Multiply_Float(A=&FNTimeOfDayScalar(Night Scale=1, Morning Scale=1, Evening Scale=1), B=0.05)
    )
    Scale RGB=&MakeVectorFromLinearColorRGB(Color=$System.TOD)
    ScaleA=true
    ScaleRGB=true
   )
   &ScaleMeshSize(Scale Factor=&VectorFromFloat(Value=&MakeCustomFloatFromBool(Bool=$User.bisHeadHidden, True Float=0, False Float=1)))
  ]
  rendererClasses [&NiagaraMeshRendererProperties]
  renderers [{rendererClass=&NiagaraMeshRendererProperties, rendererIndex=0}]
 }
}
summary {
 systemName=NS_Character_WaywardRebelFNCS_Trail
 emitters [
  {emitterName=Trail, rendererClasses [&NiagaraMeshRendererProperties]}
  {emitterName=Glow, rendererClasses [&NiagaraSpriteRendererProperties]}
  {emitterName=Trail_soft, rendererClasses [&NiagaraMeshRendererProperties]}
  {emitterName=Trail_group, rendererClasses [&NiagaraMeshRendererProperties]}
  {emitterName=Trail_soft_group, rendererClasses [&NiagaraMeshRendererProperties]}
  {emitterName=Sprites_group, rendererClasses [&NiagaraSpriteRendererProperties]}
  {emitterName=Sprites, rendererClasses [&NiagaraSpriteRendererProperties]}
  {emitterName=Light, rendererClasses [&NiagaraLightRendererProperties]}
  {emitterName=Light_group, rendererClasses [&NiagaraLightRendererProperties]}
 ]
 userVariables [User.bisHeadHidden {type=NiagaraBool}]
}
